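# Physics asset for the Binary Female pickaxe, a two-handed melee harvesting tool cosmetic.
# It holds a single body on the Melee_TwoHanded_Handle bone, shaped as one capsule roughly aligned along the handle.
# It has no constraints, so the axe simulates or collides as one rigid piece.
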PhysicsAsset /BRCosmetics/Weapons/FORT_Melee/Pickaxe_Binary_Female/Meshes/Binary_Female_Axe_PhysicsAsset
bodies {Melee_TwoHanded_Handle [{shapeType=Capsule, length=35.7188, radius=5.8857, center=<0.3508 -1.0329 24.4559>, rotation=<rot 0.6545 -90 -0.0073>}]}
constraints []
